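# Unlit master material for the Glow Rings building parts in Creative, producing a solid, tintable emissive glow.
# The Glow_Color parameter, magenta by default, is multiplied by the Glow_Mult scalar to drive Emissive Color, with no textures.
# The resulting material attributes pass through MF_CreativeMoveTool, a shared function named for the Creative move tool.
# Usage flags cover Nanite, skeletal meshes and mesh deformers.
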Material /Game/Creative/Items/Building_Parts/Glow_Rings/M_GlowRings_Master
graph {
 properties {
  bUseMaterialAttributes=true
  bUsedWithMeshDeformer=true
  bUsedWithNanite=true
  bUsedWithSkeletalMesh=true
  pixelDepthOffsetMode=PDOM_Legacy
  shadingModel=MSM_Unlit
  usedShadingModels=Unlit
 }
 nodes [
  n1 = SetMaterialAttributes!(
   MaterialAttributes
   Base Color
   Metallic
   Specular
   Roughness
   Emissive Color=Multiply(
    A=VectorParameter() {parameterName=Glow_Color, defaultValue=<rgba 1 0 1 1>}.RGB
    B=ScalarParameter() {defaultValue=1, parameterName=Glow_Mult}
   )
   Opacity Mask
   Normal
   World Position Offset
   World Displacement
   Customized UV 3
   Customized UV 4
  ) {attributeSetTypes [
    #1
    #2
    #3
    #4
    #5
    #6
    #7
    #8
    #9
    #10
    #11
   ]}
  MaterialFunctionCall(MA In=@n1) {materialFunction=&MF_CreativeMoveTool}
 ]
}
